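# Physics asset for the F_MED_DefectGlitch female medium player body cosmetic, covering only its left and right dangling straps.
# Each side has a base capsule linked to a root strap capsule with tight swing limits, which branches into three short capsule chains (A, B, C).
# The branch joints allow wider swing, while every joint keeps twist nearly locked, so the straps bend without spinning.
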PhysicsAsset /BRCosmetics/Characters/Player/Female/Medium/Bodies/F_MED_DefectGlitch/Meshes/F_MED_DefectGlitch_Physics
bodies {
 base_strap_00_l [{shapeType=Capsule, length=1, radius=1, shapeName=base_strap_00_l_capsule}]
 base_strap_00_r [{shapeType=Capsule, length=1, radius=1, shapeName=base_strap_00_r_capsule}]
 dyn_strap_00_l [{shapeType=Capsule, length=1, radius=1, shapeName=dyn_strap_00_l_capsule, center=<-1 0 0>, rotation=<rot 90 0 0>}]
 dyn_strap_00_r [{shapeType=Capsule, length=1, radius=1, shapeName=dyn_strap_00_r_capsule, center=<1 0 0>, rotation=<rot 90 0 0>}]
 dyn_strap_A_00_l [{shapeType=Capsule, length=2, radius=1, shapeName=dyn_strap_A_00_l_capsule, center=<-2 0 0>, rotation=<rot 90 0 0>}]
 dyn_strap_A_00_r [{shapeType=Capsule, length=2, radius=1, shapeName=dyn_strap_A_00_r_capsule, center=<2 0 0>, rotation=<rot 90 0 0>}]
 dyn_strap_A_01_l [{shapeType=Capsule, length=2, radius=1, shapeName=dyn_strap_A_01_l_capsule, center=<-2 0 0>, rotation=<rot 90 0 0>}]
 dyn_strap_A_01_r [{shapeType=Capsule, length=2, radius=1, shapeName=dyn_strap_A_01_r_capsule, center=<2 0 0>, rotation=<rot 90 0 0>}]
 dyn_strap_A_02_l [{shapeType=Capsule, length=2, radius=1, shapeName=dyn_strap_A_02_l_capsule, center=<-2 0 0>, rotation=<rot 90 0 0>}]
 dyn_strap_A_02_r [{shapeType=Capsule, length=2, radius=1, shapeName=dyn_strap_A_02_r_capsule, center=<2 0 0>, rotation=<rot 90 0 0>}]
 dyn_strap_B_00_l [{shapeType=Capsule, length=1, radius=1, shapeName=dyn_strap_B_00_l_capsule, center=<-1 0 0>, rotation=<rot 90 0 0>}]
 dyn_strap_B_00_r [{shapeType=Capsule, length=1, radius=1, shapeName=dyn_strap_B_00_r_capsule, center=<1 0 0>, rotation=<rot 90 0 0>}]
 dyn_strap_B_01_l [{shapeType=Capsule, length=1, radius=1, shapeName=dyn_strap_B_01_l_capsule, center=<-1 0 0>, rotation=<rot 90 0 0>}]
 dyn_strap_B_01_r [{shapeType=Capsule, length=1, radius=1, shapeName=dyn_strap_B_01_r_capsule, center=<1 0 0>, rotation=<rot 90 0 0>}]
 dyn_strap_C_00_l [{shapeType=Capsule, length=1.5, radius=1, shapeName=dyn_strap_C_00_l_capsule, center=<-1.5 0 0>, rotation=<rot 90 0 0>}]
 dyn_strap_C_00_r [{shapeType=Capsule, length=1.5, radius=1, shapeName=dyn_strap_C_00_r_capsule, center=<1.5 0 0>, rotation=<rot 90 0 0>}]
 dyn_strap_C_01_l [{shapeType=Capsule, length=1.5, radius=1, shapeName=dyn_strap_C_01_l_capsule, center=<-1.5 0 0>, rotation=<rot 90 0 0>}]
 dyn_strap_C_01_r [{shapeType=Capsule, length=1.5, radius=1, shapeName=dyn_strap_C_01_r_capsule, center=<1.5 0 0>, rotation=<rot 90 0 0>}]
}
constraints [
 {bone1Name=dyn_strap_00_r, bone2Name=base_strap_00_r, swing1LimitDegrees=15, swing2LimitDegrees=15, twistLimitDegrees=1}
 {bone1Name=dyn_strap_A_00_r, bone2Name=dyn_strap_00_r, swing1LimitDegrees=45, swing2LimitDegrees=45, twistLimitDegrees=1}
 {bone1Name=dyn_strap_A_01_r, bone2Name=dyn_strap_A_00_r, swing1LimitDegrees=45, swing2LimitDegrees=45, twistLimitDegrees=1}
 {bone1Name=dyn_strap_A_02_r, bone2Name=dyn_strap_A_01_r, swing1LimitDegrees=45, swing2LimitDegrees=45, twistLimitDegrees=1}
 {bone1Name=dyn_strap_B_00_r, bone2Name=dyn_strap_00_r, swing1LimitDegrees=45, swing2LimitDegrees=45, twistLimitDegrees=1}
 {bone1Name=dyn_strap_B_01_r, bone2Name=dyn_strap_B_00_r, swing1LimitDegrees=45, swing2LimitDegrees=45, twistLimitDegrees=1}
 {bone1Name=dyn_strap_C_00_r, bone2Name=dyn_strap_00_r, swing1LimitDegrees=45, swing2LimitDegrees=45, twistLimitDegrees=1}
 {bone1Name=dyn_strap_C_01_r, bone2Name=dyn_strap_C_00_r, swing1LimitDegrees=45, swing2LimitDegrees=45, twistLimitDegrees=1}
 {bone1Name=dyn_strap_00_l, bone2Name=base_strap_00_l, swing1LimitDegrees=15, swing2LimitDegrees=15, twistLimitDegrees=1}
 {bone1Name=dyn_strap_A_00_l, bone2Name=dyn_strap_00_l, swing1LimitDegrees=45, swing2LimitDegrees=45, twistLimitDegrees=1}
 {bone1Name=dyn_strap_A_01_l, bone2Name=dyn_strap_A_00_l, swing1LimitDegrees=45, swing2LimitDegrees=45, twistLimitDegrees=1}
 {bone1Name=dyn_strap_A_02_l, bone2Name=dyn_strap_A_01_l, swing1LimitDegrees=45, swing2LimitDegrees=45, twistLimitDegrees=1}
 {bone1Name=dyn_strap_B_00_l, bone2Name=dyn_strap_00_l, swing1LimitDegrees=45, swing2LimitDegrees=45, twistLimitDegrees=1}
 {bone1Name=dyn_strap_B_01_l, bone2Name=dyn_strap_B_00_l, swing1LimitDegrees=45, swing2LimitDegrees=45, twistLimitDegrees=1}
 {bone1Name=dyn_strap_C_00_l, bone2Name=dyn_strap_00_l, swing1LimitDegrees=45, swing2LimitDegrees=45, twistLimitDegrees=1}
 {bone1Name=dyn_strap_C_01_l, bone2Name=dyn_strap_C_00_l, swing1LimitDegrees=45, swing2LimitDegrees=45, twistLimitDegrees=1}
]
