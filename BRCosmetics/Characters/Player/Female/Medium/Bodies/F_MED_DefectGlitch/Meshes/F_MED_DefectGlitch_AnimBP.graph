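# Animation Blueprint for the F_MED_DefectGlitch player body cosmetic, layering procedural deformation and secondary physics onto the pose copied from the attached parent mesh.
# The copied pose runs through a deform Control Rig, weighted constraints tying both base straps to thigh and pelvis bones, then pelvis-space RigidBody simulation.
# Each update sets simulation gravity from character state via a deprecated helper, and after a zero-length delay feeds a spine_05-relative pivot and switch to the mesh materials.
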AnimBlueprint /BRCosmetics/Characters/Player/Female/Medium/Bodies/F_MED_DefectGlitch/Meshes/F_MED_DefectGlitch_AnimBP
target_skeleton=&/F_MED_DefectGlitch_Skeleton
graphs [
 AnimGraph {nodes [
   n1 = Control Rig(Source=Copy Pose From Mesh() {bUseAttachedParent=true}) {
    alphaBoolBlend {blendOption=Linear}
    controlRigAssetReference {blueprintRigClass=&/F_MED_DefectGlitch_DeformRig_Master_ControlRig.F_MED_DefectGlitch_DeformRig_Master_ControlRig_C}
   }
   n2 = "Constraint - base_strap_00_r (PPP)"!(
    ComponentPose=Local To Component(LocalPose=@n1)
    Alpha=1
    ConstraintWeights_0=0.5738
    ConstraintWeights_1:float=0.4253
    ConstraintWeights_2:float=0.001
   ) -> (Pose) {
    alphaBoolBlend {blendOption=Linear}
    boneToModify {boneName=base_strap_00_r}
    constraintSetup [{targetBone {boneName=thigh_twist_01_r}}, {targetBone {boneName=thigh_r}}, {targetBone {boneName=pelvis}}]
    constraintWeights [1,1,1]
   }
   n3 = "Constraint - base_strap_00_l (PPP)"!(ComponentPose=@n2, Alpha=1, ConstraintWeights_0=0.5761, ConstraintWeights_1:float=0.4231, ConstraintWeights_2:float=0.0008) -> (Pose) {
    alphaBoolBlend {blendOption=Linear}
    boneToModify {boneName=base_strap_00_l}
    constraintSetup [{targetBone {boneName=thigh_twist_01_l}}, {targetBone {boneName=thigh_l}}, {targetBone {boneName=pelvis}}]
    constraintWeights [1,1,1]
   }
   n4 = RigidBody!(ComponentPose=@n3, bAlphaBoolEnabled:bool, OverrideWorldGravity:Vector=Get GravityOverride(), ExternalForce) -> (Pose) {
    alphaInputType=Bool
    bClampLinearTranslationLimitToRefPose=true
    bOverrideWorldGravity=true
    overridePhysicsAsset=&/F_MED_DefectGlitch_Physics
    simulationSpace=BaseBoneSpace
    alphaBoolBlend {blendOption=Linear}
    baseBoneRef {boneName=pelvis}
    componentAppliedLinearAccClamp=<500 500 500>
    componentLinearAccScale=<0.5 0.3 0.3>
    componentLinearVelScale=<0.3 0.3 0.3>
    overrideWorldGravity {z=-980}
    simSpaceSettings {
     maxAngularAcceleration=100
     maxAngularVelocity=0.5
     maxLinearAcceleration=200
     maxLinearVelocity=200
     velocityScaleZ=0
     worldAlpha=1
     externalLinearDragV=<0.2 0.2>
    }
   }
   Output Pose(Result=Component To Local(ComponentPose=@n4)) {layerGroup=None}
  ]}
 EventGraph {nodes [
   Event BlueprintUpdateAnimation()
   > n1 = Sequence()
   > n2 = Gravity Override State Function DEPRECATED(
    CharacterPartAnimInstance=Self-Reference()
    GameTimeInSeconds=GetGameTimeInSeconds()
    GravityOverrideParams=Get GravityOverrideParams() -> (GravityOverrideParams:Gravity Override Params Struct Structure)
   )
   > Set GravityOverride(GravityOverride=@n2)
   n3 = GetPartSkeletalMeshComponent()
   Is Valid(exec=@n1.then_1, InputObject=@n3)
   > Delay(Duration=0)
   > SetVectorParameterValueOnMaterials!(
    execute
    self:Mesh Component Object Reference=@n3
    ParameterName:Name=DO NOT OVERRIDE-Virtual-Pivot-Ch
    ParameterValue:Vector=TransformLocation!(T:"Transform (by ref)"=GetSocketTransform(self=@n3, InSocketName=spine_05, TransformSpace=RTS_World), Location:Vector="-5, 0,0") -> (ReturnValue:Vector)
   ) -> (then)
   > SetScalarParameterValueOnMaterials!(execute, self:Mesh Component Object Reference=@n3, ParameterName:Name=DO NOT OVERRIDE-Character-Position-BP-Switch, ParameterValue:float=1) -> (then)
  ]}
]
variables [
 GravityOverride {type=Vector}
 GravityOverrideParams {
  type=GravityOverrideParamsStruct
  default {
   emoteOrMelee {gravityOverride=<0 0 -700>}
   front_End {gravityOverride=<0 0 -800>}
   onGround_Falling {gravityOverride=<0 0 -800>}
   onGround_Moving {gravityOverride=<0 0 -800>}
   onGround_Standing {gravityOverride=<0 0 -850>}
  }
 }
]
