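# Control Rig for the DefectGlitch female face accessory that repositions back hair base bones as the head and right clavicle move.
# A sequence runs three spherical pose readers, whose outputs weight Modify Transforms nodes using offsets stored in variables.
# One head reader rotates the outer right and both middle locks, the latter capped at half weight.
# The clavicle reader rotates both right locks, and a second head reader shifts the outer right one.
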ControlRigBlueprint /BRCosmetics/Characters/Player/Female/Medium/Bodies/F_MED_DefectGlitch/Meshes/Parts/F_MED_DefectGlitch_FaceAcc_CtrlRig
graphs [
 ContainedGraph {nodes [
   n1 = RigVMFunctionEntryNode()
   SequenceAggregate = &RigVMFunction_Sequence(ExecuteContext=@n1.ExecuteContext)
   SequenceAggregate_1 = &RigVMFunction_Sequence(ExecuteContext=@SequenceAggregate.B)
   RigVMFunctionReturnNode(A=@SequenceAggregate.A, B=@SequenceAggregate_1.A, C=@SequenceAggregate_1.B)
  ]}
 RigVMFunctionLibrary {links [], nodes []}
 RigVMModel {nodes [
   RigUnit_BeginExecution = &RigUnit_BeginExecution()
   SequenceAggregate = RigVMAggregateNode(ExecuteContext=@RigUnit_BeginExecution.ExecutePin)
   n1 = &RigUnit_SphericalPoseReader(
    ActiveRegionScaleFactors=(PositiveWidth=1,NegativeWidth=1,PositiveHeight=1,NegativeHeight=1)
    ActiveRegionSize=0.1
    Debug=(bDrawDebug=True,bDraw2D=False,bDrawLocalAxes=False,DebugScale=25,DebugSegments=20,DebugThickness=0.25)
    DriverAxis=(X=1,Y=0,Z=0)
    DriverItem=(Type=Bone,Name="head")
    FalloffRegionScaleFactors=(PositiveWidth=1,NegativeWidth=1,PositiveHeight=1,NegativeHeight=1)
    FalloffSize=0.18
    FlipHeightScaling=false
    FlipWidthScaling=false
    OptionalParentItem=(Type=Bone)
    RotationOffset=(X=90,Y=0,Z=-125)
    ExecutePin=@SequenceAggregate.A
   )
   > &RigUnit_ModifyTransforms(
    ItemToModify=((Item=(Type=Bone,Name="base_hair_back_a_00_r"),Transform=(Rotation=(X=0,Y=0,Z=0,W=1),Translation=(X=0,Y=0,Z=0),Scale3D=(X=1,Y=1,Z=1))))
    Weight=@n1.OutputParam
    ItemToModify.0.Transform=RigVMVariableNode() {variable=Hair Out R Transform}.Value
   )
   > &RigUnit_ModifyTransforms(
    ItemToModify=((Item=(Type=Bone,Name="base_hair_back_b_00_r"),Transform=(Rotation=(X=0,Y=0,Z=0,W=1),Translation=(X=0,Y=0,Z=0),Scale3D=(X=1,Y=1,Z=1))))
    WeightMaximum=0.5
    ItemToModify.0.Transform=RigVMVariableNode() {variable=Hair Middle R Transform}.Value
    Weight=@n1.OutputParam
   )
   > &RigUnit_ModifyTransforms(
    ItemToModify=((Item=(Type=Bone,Name="base_hair_back_b_00_l"),Transform=(Rotation=(X=0,Y=0,Z=0,W=1),Translation=(X=0,Y=0,Z=0),Scale3D=(X=1,Y=1,Z=1))))
    WeightMaximum=0.5
    ItemToModify.0.Transform=RigVMVariableNode() {variable=Hair Middle L Transform}.Value
    Weight=@n1.OutputParam
   )
   n2 = &RigUnit_SphericalPoseReader(
    ActiveRegionScaleFactors=(PositiveWidth=1,NegativeWidth=1,PositiveHeight=1,NegativeHeight=1)
    ActiveRegionSize=0.1
    Debug=(bDrawDebug=True,bDraw2D=False,bDrawLocalAxes=False,DebugScale=25,DebugSegments=20,DebugThickness=0.25)
    DriverAxis=(X=-1,Y=0,Z=0)
    DriverItem=(Type=Bone,Name="clavicle_r")
    FalloffRegionScaleFactors=(PositiveWidth=1,NegativeWidth=1,PositiveHeight=1,NegativeHeight=1)
    FalloffSize=0.18
    FlipHeightScaling=false
    FlipWidthScaling=false
    OptionalParentItem=(Type=Bone)
    RotationOffset=(X=0,Y=125,Z=0)
    ExecutePin=@SequenceAggregate.B
   )
   n3 = RigVMVariableNode() {variable=Hair Out R Transform_0}
   > &RigUnit_ModifyTransforms(
    ItemToModify=((Item=(Type=Bone,Name="base_hair_back_a_00_r"),Transform=(Rotation=(X=0,Y=0,Z=0,W=1),Translation=(X=0,Y=0,Z=0),Scale3D=(X=1,Y=1,Z=1))))
    Weight=@n2.OutputParam
    ItemToModify.0.Transform=@n3.Value
   )
   > &RigUnit_ModifyTransforms(
    ItemToModify=((Item=(Type=Bone,Name="base_hair_back_b_00_r"),Transform=(Rotation=(X=0,Y=0,Z=0,W=1),Translation=(X=0,Y=0,Z=0),Scale3D=(X=1,Y=1,Z=1))))
    Weight=@n2.OutputParam
    ItemToModify.0.Transform=@n3.Value
   )
   n4 = &RigUnit_SphericalPoseReader(
    ActiveRegionScaleFactors=(PositiveWidth=1,NegativeWidth=1,PositiveHeight=1,NegativeHeight=1)
    ActiveRegionSize=0.1
    Debug=(bDrawDebug=True,bDraw2D=False,bDrawLocalAxes=False,DebugScale=25,DebugSegments=20,DebugThickness=0.25)
    DriverAxis=(X=1,Y=0,Z=0)
    DriverItem=(Type=Bone,Name="head")
    FalloffRegionScaleFactors=(PositiveWidth=1,NegativeWidth=1,PositiveHeight=1,NegativeHeight=1)
    FalloffSize=0.18
    FlipHeightScaling=false
    FlipWidthScaling=false
    OptionalParentItem=(Type=Bone)
    RotationOffset=(X=90,Y=0,Z=-50)
    ExecutePin=@SequenceAggregate.C
   )
   > &RigUnit_ModifyTransforms(
    ItemToModify=((Item=(Type=Bone,Name="base_hair_back_a_00_r"),Transform=(Rotation=(X=0,Y=0,Z=0,W=1),Translation=(X=0,Y=0,Z=0),Scale3D=(X=1,Y=1,Z=1))))
    Weight=@n4.OutputParam
    ItemToModify.0.Transform=RigVMVariableNode() {variable=Hair Out R Transform_1}.Value
   )
  ]}
]
hierarchy [
 root {loc=<0 0 0>, rot=<0 0 0 1>}
 pelvis {parent=root, loc=<0 -0.5301 98.0571>, rot=<0.0127 0.707 -0.0127 -0.707>}
 spine_01 {parent=pelvis, loc=<6.1208 -0.9328 0>, rot=<0 0 0.034 -0.9994>}
 spine_02 {parent=spine_01, loc=<6.5996 -0.2095 0>, rot=<0 0 0.0238 -0.9997>}
 spine_03 {parent=spine_02, loc=<6.4914 -0.1237 0>, rot=<0 0 0.0149 -0.9999>}
 spine_04 {parent=spine_03, loc=<6.9598 0.0794 0>, rot=<0 0 -0.0132 -0.9999>}
 spine_05 {parent=spine_04, loc=<6.9987 0 0>, rot=<0 0 -0.0015 -1>}
 clavicle_l {parent=spine_05, loc=<5.4831 -0.1053 -3.2946>, rot=<0.0465 -0.7289 -0.009 -0.6829>}
 neck_01 {parent=spine_05, loc=<11.3138 0.1529 0>, rot=<0 0 -0.2003 -0.9797>}
 neck_02 {parent=neck_01, loc=<4.5627 0 0>, rot=<0 0 0.0504 -0.9987>}
 head {parent=neck_02, loc=<5.3914 0 0>, rot=<0 0 0.115 -0.9934>}
 base_earing_l {parent=head, loc=<0.2163 0.2516 -6.1724>, rot=<0.0176 0.4836 0.8749 -0.0206>}
 dyn_earing_l {parent=base_earing_l, loc=<0 0 0>, rot=<0 0 0 1>}
 base_earing_r {parent=head, loc=<0.2164 0.2516 6.1724>, rot=<0.0206 -0.8749 -0.4836 -0.0176>}
 dyn_earing_r {parent=base_earing_r, loc=<0 0 0>, rot=<0 0 0 1>}
 base_hair_back_a_00_l {parent=head, loc=<2.099 -4.364 -7.174>, rot=<-0.0678 -0.9893 -0.0088 -0.1286>}
 dyn_hair_back_a_00_l {parent=base_hair_back_a_00_l, loc=<0 0 0>, rot=<0 0 0 1>}
 dyn_hair_back_a_01_l {parent=dyn_hair_back_a_00_l, loc=<3.6425 0 0>, rot=<0 -0.0074 0 -1>}
 base_hair_back_b_00_l {parent=head, loc=<2.1376 -8.5915 -3.2535>, rot=<0.0592 -0.7849 0.6027 -0.1311>}
 dyn_hair_back_b_00_l {parent=base_hair_back_b_00_l, loc=<0 0 0>, rot=<0 0 0 1>}
 dyn_hair_back_b_01_l {parent=dyn_hair_back_b_00_l, loc=<2.8427 0 0>, rot=<0 -0.0897 0 -0.996>}
 base_hair_back_a_00_r {parent=head, loc=<1.9051 -3.7906 7.1001>, rot=<0.1038 0.0044 0.994 0.0336>}
 dyn_hair_back_a_00_r {parent=base_hair_back_a_00_r, loc=<0 0 0>, rot=<0 0 0 1>}
 dyn_hair_back_a_01_r {parent=dyn_hair_back_a_00_r, loc=<4.6524 0 0>, rot=<0 -0.0298 0 -0.9996>}
 base_hair_back_b_00_r {parent=head, loc=<2.5691 -8.2209 2.6193>, rot=<0.1104 -0.6918 0.6983 -0.1469>}
 dyn_hair_back_b_00_r {parent=base_hair_back_b_00_r, loc=<0 0 0>, rot=<0 0 0 1>}
 dyn_hair_back_b_01_r {parent=dyn_hair_back_b_00_r, loc=<3.7752 0 0>, rot=<0 -0.0583 0 -0.9983>}
 base_hair_front_c_00_r {parent=head, loc=<7.4523 5.7966 6.1182>, rot=<0.0272 0.1633 0.9691 0.1826>}
 dyn_hair_front_c_00_r {parent=base_hair_front_c_00_r, loc=<0 0 0>, rot=<0 0 0 1>}
 base_hair_side_b_00_l {parent=head, loc=<1.5173 1.8573 -6.9394>, rot=<0.0199 -0.996 -0.0667 -0.0559>}
 dyn_hair_side_b_00_l {parent=base_hair_side_b_00_l, loc=<0 0 0>, rot=<0 0 0 1>}
 dyn_hair_side_b_01_l {parent=dyn_hair_side_b_00_l, loc=<4.1331 -0.1343 0.2768>, rot=<-0.0001 0.0747 -0.0017 -0.9972>}
 base_hair_side_a_00_l {parent=head, loc=<0.893 5.5113 -6.3502>, rot=<0.0483 -0.9981 0.0015 -0.0373>}
 dyn_hair_side_a_00_l {parent=base_hair_side_a_00_l, loc=<0 0 0>, rot=<0 0 0 1>}
 base_hair_front_b_00_r {parent=head, loc=<9.1395 9.4218 1.2925>, rot=<0.2056 0.3806 0.8994 0.063>}
 dyn_hair_front_b_00_r {parent=base_hair_front_b_00_r, loc=<0 0 0>, rot=<0 0 0 1>}
 dyn_hair_front_b_01_r {parent=dyn_hair_front_b_00_r, loc=<2.6061 0 0>, rot=<-0.011 -0.169 0.064 -0.9835>}
 dyn_hair_front_b_02_r {parent=dyn_hair_front_b_01_r, loc=<2.8107 0 0>, rot=<-0.0055 -0.0897 0.0615 -0.9941>}
 base_hair_front_a_00_r {parent=head, loc=<10.5072 9.8335 0.7563>, rot=<0.1887 0.5306 0.8237 0.0661>}
 dyn_hair_front_a_00_r {parent=base_hair_front_a_00_r, loc=<0 0 0>, rot=<0 0 0 1>}
 dyn_hair_front_a_01_r {parent=dyn_hair_front_a_00_r, loc=<3.0329 0 0>, rot=<0.0975 -0.1386 0.04 -0.9847>}
 dyn_hair_front_a_02_r {parent=dyn_hair_front_a_01_r, loc=<4.0087 -0.1463 -0.2562>, rot=<0.0317 -0.1016 0.0538 -0.9929>}
 dyn_hair_front_a_03_r {parent=dyn_hair_front_a_02_r, loc=<2.7328 -0.0461 -0.2019>, rot=<-0.0025 -0.0849 0.0288 -0.996>}
 base_hair_front_a_00_l {parent=head, loc=<10.175 10.0745 -1.5769>, rot=<-0.0333 -0.8436 -0.4152 -0.3388>}
 dyn_hair_front_a_00_l {parent=base_hair_front_a_00_l, loc=<0 0 0>, rot=<0 0 0 1>}
 dyn_hair_front_a_01_l {parent=dyn_hair_front_a_00_l, loc=<2.193 0 0>, rot=<-0.0511 -0.2409 -0.0027 -0.9692>}
 dyn_hair_front_a_02_l {parent=dyn_hair_front_a_01_l, loc=<3.0093 -0.1201 0.0437>, rot=<-0.0018 -0.1361 0.0128 -0.9906>}
 dyn_hair_front_a_03_l {parent=dyn_hair_front_a_02_l, loc=<2.5915 0.024 -0.0762>, rot=<0.0035 -0.1203 -0.0287 -0.9923>}
 clavicle_r {parent=spine_05, loc=<5.4831 -0.1053 3.2946>, rot=<0.7289 0.0465 -0.6829 0.009>}
]
variables [
 Hair Middle L Transform {type=FTransform, default=(Rotation=(X=0,Y=-0.1908,Z=0,W=0.9816),Translation=(X=0,Y=0,Z=0),Scale3D=(X=1,Y=1,Z=1))}
 Hair Middle R Transform {type=FTransform, default=(Rotation=(X=-0.2541,Y=-0.1843,Z=-0.0494,W=0.9482),Translation=(X=0,Y=0,Z=0),Scale3D=(X=1,Y=1,Z=1))}
 Hair Out R Transform {type=FTransform, default=(Rotation=(X=-0.1913,Y=-0.4619,Z=0.1913,W=0.8446),Translation=(X=0,Y=0,Z=0),Scale3D=(X=1,Y=1,Z=1))}
 Hair Out R Transform_0 {type=FTransform, default=(Rotation=(X=-0.0782,Y=-0.0579,Z=0.1691,W=0.9808),Translation=(X=0,Y=0,Z=0),Scale3D=(X=1,Y=1,Z=1))}
 Hair Out R Transform_1 {type=FTransform, default=(Rotation=(X=0,Y=0,Z=0,W=1),Translation=(X=0,Y=1.25,Z=0),Scale3D=(X=1,Y=1,Z=1))}
]
