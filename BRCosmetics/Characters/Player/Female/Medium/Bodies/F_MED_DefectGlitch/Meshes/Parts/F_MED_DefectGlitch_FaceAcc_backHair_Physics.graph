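# Physics asset for the back-hair part of the DefectGlitch female medium player character, defining capsule bodies and joints for hair physics.
# Four hair chains (a/b, left/right) each link a base bone at the head to two dynamic bones with limited swing and twist.
# Base-to-head joints differ: two are free, two nearly rigid.
# Head, neck and clavicle capsules add body volumes; several right-side bones use multiple capsules.
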PhysicsAsset /BRCosmetics/Characters/Player/Female/Medium/Bodies/F_MED_DefectGlitch/Meshes/Parts/F_MED_DefectGlitch_FaceAcc_backHair_Physics
bodies {
 base_hair_back_a_00_l [{shapeType=Capsule, length=0, radius=1, shapeName=base_hair_back_a_00_l_capsule, rotation=<rot 90 0 0>}]
 base_hair_back_a_00_r [{shapeType=Capsule, length=0, radius=1, shapeName=base_hair_back_a_00_r_capsule, rotation=<rot 90 0 0>}]
 base_hair_back_b_00_l [{shapeType=Capsule, length=0, radius=1, shapeName=base_hair_back_b_00_l_capsule, rotation=<rot 90 0 0>}]
 base_hair_back_b_00_r [{shapeType=Capsule, length=0, radius=1, shapeName=base_hair_back_b_00_r_capsule, rotation=<rot 90 0 0>}]
 clavicle_l [{shapeType=Capsule, length=8, radius=6, shapeName=clavicle_l_capsule, center=<4 1 -0.5>, rotation=<rot 90 0 0>}]
 clavicle_r [
  {shapeType=Capsule, length=8, radius=6, shapeName=clavicle_r_capsule, center=<-4 -3 0.5>, rotation=<rot 90 0 0>}
  {shapeType=Capsule, length=8, radius=6, shapeName=clavicle_r_capsule, center=<-4 0.5 0.5>, rotation=<rot 90 0 0>}
 ]
 dyn_hair_back_a_00_l [{shapeType=Capsule, length=1.25, radius=1, shapeName=dyn_hair_back_a_00_l_capsule, center=<2.098 0 0>, rotation=<rot 90 0 0>}]
 dyn_hair_back_a_00_r [{shapeType=Capsule, length=3, radius=2, shapeName=dyn_hair_back_a_00_r_capsule, center=<2 0 0>, rotation=<rot 90 0 0>}]
 dyn_hair_back_a_01_l [{shapeType=Capsule, length=1.5, radius=1.5, shapeName=dyn_hair_back_a_01_l_capsule, center=<4.2117 0 0>, rotation=<rot 90 -165.9638 104.0362>}]
 dyn_hair_back_a_01_r [
  {shapeType=Capsule, length=2, radius=2, shapeName=dyn_hair_back_a_01_r_capsule, center=<3 0 0>, rotation=<rot 90 0 0>}
  {
   shapeType=Capsule
   length=2.5
   radius=1
   shapeName=dyn_hair_back_a_01_r_capsule
   center=<4.1857 -1.9541 -0.7147>
   rotation=<rot 80 -0.0001 16.6397>
  }
  {
   shapeType=Capsule
   length=2.5
   radius=1.56
   shapeName=dyn_hair_back_a_01_r_capsule
   center=<3.1461 2.6929 -0.7823>
   rotation=<rot 63.081 93.4707 159.1559>
  }
 ]
 dyn_hair_back_b_00_l [{shapeType=Capsule, length=1, radius=1, shapeName=dyn_hair_back_b_00_l_capsule, center=<2.3356 0 0>, rotation=<rot 90 0 0>}]
 dyn_hair_back_b_00_r [{shapeType=Capsule, length=1, radius=1, shapeName=dyn_hair_back_b_00_r_capsule, center=<1 0 0>, rotation=<rot 90 0 0>}]
 dyn_hair_back_b_01_l [{
   shapeType=Capsule
   length=2.25
   radius=1.25
   shapeName=dyn_hair_back_b_01_l_capsule
   center=<2.9111 0 0>
   rotation=<rot 90 -165.9638 104.0362>
  }]
 dyn_hair_back_b_01_r [
  {shapeType=Capsule, length=2, radius=2, shapeName=dyn_hair_back_b_01_r_capsule, center=<2 0 0>, rotation=<rot 90 90 0>}
  {
   shapeType=Capsule
   length=4
   radius=2
   shapeName=dyn_hair_back_b_01_r_capsule
   center=<2.7883 -2.1216 -0.4692>
   rotation=<rot 68.5707 -89.9997 -179.9997>
  }
 ]
 head [{shapeType=Capsule, length=0, radius=5.5}]
 neck_01 [{shapeType=Capsule, length=4, radius=4.5, shapeName=neck_01_capsule, center=<2 0 0>, rotation=<rot 90 0 0>}]
}
constraints [
 {
  bone1Name=base_hair_back_a_00_l
  bone2Name=head
  swing1LimitDegrees=45
  swing1Motion=Free
  swing2LimitDegrees=45
  swing2Motion=Free
  twistLimitDegrees=45
  twistMotion=Free
 }
 {bone1Name=dyn_hair_back_a_00_l, bone2Name=base_hair_back_a_00_l, swing1LimitDegrees=20, swing2LimitDegrees=20, twistLimitDegrees=5}
 {bone1Name=dyn_hair_back_a_01_l, bone2Name=dyn_hair_back_a_00_l, swing1LimitDegrees=45, swing2LimitDegrees=45, twistLimitDegrees=5}
 {bone1Name=base_hair_back_a_00_r, bone2Name=head, swing1LimitDegrees=1, swing2LimitDegrees=1, twistLimitDegrees=1}
 {bone1Name=dyn_hair_back_a_00_r, bone2Name=base_hair_back_a_00_r, swing1LimitDegrees=40, swing2LimitDegrees=40, twistLimitDegrees=20}
 {bone1Name=dyn_hair_back_a_01_r, bone2Name=dyn_hair_back_a_00_r, swing1LimitDegrees=10, swing2LimitDegrees=8, twistLimitDegrees=1}
 {bone1Name=base_hair_back_b_00_l, bone2Name=head, swing1LimitDegrees=1, swing2LimitDegrees=1, twistLimitDegrees=1}
 {bone1Name=dyn_hair_back_b_00_l, bone2Name=base_hair_back_b_00_l, swing1LimitDegrees=20, swing2LimitDegrees=30, twistLimitDegrees=1}
 {bone1Name=dyn_hair_back_b_01_l, bone2Name=dyn_hair_back_b_00_l, swing1LimitDegrees=45, swing2LimitDegrees=45, twistLimitDegrees=1}
 {
  bone1Name=base_hair_back_b_00_r
  bone2Name=head
  swing1LimitDegrees=45
  swing1Motion=Free
  swing2LimitDegrees=45
  swing2Motion=Free
  twistLimitDegrees=45
  twistMotion=Free
 }
 {bone1Name=dyn_hair_back_b_00_r, bone2Name=base_hair_back_b_00_r, swing1LimitDegrees=20, swing2LimitDegrees=30, twistLimitDegrees=1}
 {bone1Name=dyn_hair_back_b_01_r, bone2Name=dyn_hair_back_b_00_r, swing1LimitDegrees=15, swing2LimitDegrees=30, twistLimitDegrees=1}
 {
  bone1Name=head
  bone2Name=neck_01
  swing1LimitDegrees=45
  swing1Motion=Free
  swing2LimitDegrees=45
  swing2Motion=Free
  twistLimitDegrees=45
  twistMotion=Free
 }
]
