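# Physics asset for the earrings of the F_MED_DefectGlitch female medium character's face accessory part.
# Capsule bodies cover the head plus a base and a dangling "dyn" bone for each earring, chained head to base to dangle by constraints.
# Base-to-head joints set swing and twist free, while dangle-to-base joints carry angular limits with very tight twist.
# Those limits are asymmetric: the left earring's first swing is far narrower than the right's.
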PhysicsAsset /BRCosmetics/Characters/Player/Female/Medium/Bodies/F_MED_DefectGlitch/Meshes/Parts/F_MED_DefectGlitch_FaceAcc_earings_Physics
bodies {
 base_earing_l [{shapeType=Capsule, length=0.505, radius=0.505, shapeName=base_earing_l_capsule, rotation=<rot 0 -90 0>}]
 base_earing_r [{shapeType=Capsule, length=0.505, radius=0.505, shapeName=base_earing_r_capsule, rotation=<rot 0 -90 0>}]
 dyn_earing_l [{shapeType=Capsule, length=2.5, radius=1, shapeName=dyn_earing_l_capsule, center=<3 0 0>, rotation=<rot 90 0 0>}]
 dyn_earing_r [{shapeType=Capsule, length=2.5, radius=1, shapeName=dyn_earing_r_capsule, center=<3 0 0>, rotation=<rot 90 0 0>}]
 head [{shapeType=Capsule, length=0, radius=2.5}]
}
constraints [
 {
  bone1Name=base_earing_l
  bone2Name=head
  swing1LimitDegrees=45
  swing1Motion=Free
  swing2LimitDegrees=45
  swing2Motion=Free
  twistLimitDegrees=45
  twistMotion=Free
 }
 {
  bone1Name=base_earing_r
  bone2Name=head
  swing1LimitDegrees=45
  swing1Motion=Free
  swing2LimitDegrees=45
  swing2Motion=Free
  twistLimitDegrees=45
  twistMotion=Free
 }
 {bone1Name=dyn_earing_l, bone2Name=base_earing_l, swing1LimitDegrees=5, swing2LimitDegrees=45, twistLimitDegrees=1}
 {bone1Name=dyn_earing_r, bone2Name=base_earing_r, swing1LimitDegrees=45, swing2LimitDegrees=45, twistLimitDegrees=1}
]
