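# Physics asset for the front-hair face accessory of the DefectGlitch female medium player body, defining bodies and joints for front and side hair strands.
# Six chains of one to four dyn_hair capsules each hang from a small base capsule linked to a rounded head body.
# Base-to-head joints leave rotation free, while strand joints limit swing and nearly lock twist, with one swing axis loosening down the longer front chains.
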PhysicsAsset /BRCosmetics/Characters/Player/Female/Medium/Bodies/F_MED_DefectGlitch/Meshes/Parts/F_MED_DefectGlitch_FaceAcc_frontHair_Physics
bodies {
 base_hair_front_a_00_l [{shapeType=Capsule, length=0.505, radius=0.505, shapeName=base_hair_front_a_00_l_capsule, rotation=<rot 0 -90 0>}]
 base_hair_front_a_00_r [{shapeType=Capsule, length=0.505, radius=0.505, shapeName=base_hair_front_a_00_r_capsule, rotation=<rot 0 -90 0>}]
 base_hair_front_b_00_r [{shapeType=Capsule, length=0.505, radius=0.505, shapeName=base_hair_front_b_00_r_capsule, rotation=<rot 0 -90 0>}]
 base_hair_front_c_00_r [{shapeType=Capsule, length=0.505, radius=0.505, shapeName=base_hair_front_c_00_r_capsule, rotation=<rot 0 -90 0>}]
 base_hair_side_a_00_l [{shapeType=Capsule, length=0.505, radius=0.505, shapeName=base_hair_side_a_00_l_capsule, rotation=<rot 0 -90 0>}]
 base_hair_side_b_00_l [{shapeType=Capsule, length=0.505, radius=0.505, shapeName=base_hair_side_b_00_l_capsule, rotation=<rot 0 -90 0>}]
 dyn_hair_front_a_00_l [{shapeType=Capsule, length=1, radius=0.8, shapeName=dyn_hair_front_a_00_l_capsule, center=<1 0 0>, rotation=<rot 90 0 0>}]
 dyn_hair_front_a_00_r [{shapeType=Capsule, length=1, radius=0.5, shapeName=dyn_hair_front_a_00_r_capsule, center=<0.5 0 0>, rotation=<rot 90 0 0>}]
 dyn_hair_front_a_01_l [{shapeType=Capsule, length=1, radius=0.8, shapeName=dyn_hair_front_a_01_l_capsule, center=<1 0 0>, rotation=<rot 90 0 0>}]
 dyn_hair_front_a_01_r [{shapeType=Capsule, length=1, radius=0.5, shapeName=dyn_hair_front_a_01_r_capsule, center=<0.5 0 0>, rotation=<rot 90 0 0>}]
 dyn_hair_front_a_02_l [{shapeType=Capsule, length=1, radius=0.8, shapeName=dyn_hair_front_a_02_l_capsule, center=<1 0 0>, rotation=<rot 90 0 0>}]
 dyn_hair_front_a_02_r [{shapeType=Capsule, length=1, radius=0.5, shapeName=dyn_hair_front_a_02_r_capsule, center=<0.5 0 0>, rotation=<rot 90 0 0>}]
 dyn_hair_front_a_03_l [{shapeType=Capsule, length=1, radius=0.8, shapeName=dyn_hair_front_a_03_l_capsule, center=<1 0 0>, rotation=<rot 90 0 0>}]
 dyn_hair_front_a_03_r [{shapeType=Capsule, length=1, radius=0.5, shapeName=dyn_hair_front_a_03_r_capsule, center=<0.5 0 0>, rotation=<rot 90 0 0>}]
 dyn_hair_front_b_00_r [{shapeType=Capsule, length=1, radius=0.5, shapeName=dyn_hair_front_b_00_r_capsule, center=<0.5 0 0>, rotation=<rot 90 0 0>}]
 dyn_hair_front_b_01_r [{shapeType=Capsule, length=1, radius=0.5, shapeName=dyn_hair_front_b_01_r_capsule, center=<0.5 0 0>, rotation=<rot 90 0 0>}]
 dyn_hair_front_b_02_r [{shapeType=Capsule, length=1, radius=0.5, shapeName=dyn_hair_front_b_02_r_capsule, center=<0.5 0 0>, rotation=<rot 90 0 0>}]
 dyn_hair_front_c_00_r [{shapeType=Capsule, length=1, radius=1, shapeName=dyn_hair_front_c_00_r_capsule, center=<2 0 0>, rotation=<rot 90 0 0>}]
 dyn_hair_side_a_00_l [{shapeType=Capsule, length=2, radius=1, shapeName=dyn_hair_side_a_00_l_capsule, center=<2 0 0>, rotation=<rot 90 0 0>}]
 dyn_hair_side_b_00_l [{shapeType=Capsule, length=1, radius=1, shapeName=dyn_hair_side_b_00_l_capsule, center=<1 0 0>, rotation=<rot 90 0 0>}]
 dyn_hair_side_b_01_l [{shapeType=Capsule, length=1, radius=1, shapeName=dyn_hair_side_b_01_l_capsule, center=<1 0 0>, rotation=<rot 90 0 0>}]
 head [{shapeType=Capsule, length=0, radius=5.5}]
}
constraints [
 {
  bone1Name=base_hair_front_c_00_r
  bone2Name=head
  swing1LimitDegrees=45
  swing1Motion=Free
  swing2LimitDegrees=45
  swing2Motion=Free
  twistLimitDegrees=45
  twistMotion=Free
 }
 {bone1Name=dyn_hair_front_c_00_r, bone2Name=base_hair_front_c_00_r, swing1LimitDegrees=0, swing2LimitDegrees=15, twistLimitDegrees=0}
 {
  bone1Name=base_hair_front_a_00_r
  bone2Name=head
  swing1LimitDegrees=45
  swing1Motion=Free
  swing2LimitDegrees=45
  swing2Motion=Free
  twistLimitDegrees=45
  twistMotion=Free
 }
 {bone1Name=dyn_hair_front_a_00_r, bone2Name=base_hair_front_a_00_r, swing1LimitDegrees=20, swing2LimitDegrees=10, twistLimitDegrees=1}
 {bone1Name=dyn_hair_front_a_01_r, bone2Name=dyn_hair_front_a_00_r, swing1LimitDegrees=20, swing2LimitDegrees=15, twistLimitDegrees=1}
 {bone1Name=dyn_hair_front_a_02_r, bone2Name=dyn_hair_front_a_01_r, swing1LimitDegrees=20, swing2LimitDegrees=20, twistLimitDegrees=1}
 {bone1Name=dyn_hair_front_a_03_r, bone2Name=dyn_hair_front_a_02_r, swing1LimitDegrees=20, swing2LimitDegrees=20, twistLimitDegrees=1}
 {
  bone1Name=base_hair_front_b_00_r
  bone2Name=head
  swing1LimitDegrees=45
  swing1Motion=Free
  swing2LimitDegrees=45
  swing2Motion=Free
  twistLimitDegrees=45
  twistMotion=Free
 }
 {bone1Name=dyn_hair_front_b_00_r, bone2Name=base_hair_front_b_00_r, swing1LimitDegrees=20, swing2LimitDegrees=10, twistLimitDegrees=1}
 {bone1Name=dyn_hair_front_b_01_r, bone2Name=dyn_hair_front_b_00_r, swing1LimitDegrees=20, swing2LimitDegrees=15, twistLimitDegrees=1}
 {bone1Name=dyn_hair_front_b_02_r, bone2Name=dyn_hair_front_b_01_r, swing1LimitDegrees=20, swing2LimitDegrees=20, twistLimitDegrees=1}
 {
  bone1Name=base_hair_front_a_00_l
  bone2Name=head
  swing1LimitDegrees=45
  swing1Motion=Free
  swing2LimitDegrees=45
  swing2Motion=Free
  twistLimitDegrees=45
  twistMotion=Free
 }
 {bone1Name=dyn_hair_front_a_00_l, bone2Name=base_hair_front_a_00_l, swing1LimitDegrees=20, swing2LimitDegrees=10, twistLimitDegrees=1}
 {bone1Name=dyn_hair_front_a_01_l, bone2Name=dyn_hair_front_a_00_l, swing1LimitDegrees=20, swing2LimitDegrees=15, twistLimitDegrees=1}
 {bone1Name=dyn_hair_front_a_02_l, bone2Name=dyn_hair_front_a_01_l, swing1LimitDegrees=20, swing2LimitDegrees=20, twistLimitDegrees=1}
 {bone1Name=dyn_hair_front_a_03_l, bone2Name=dyn_hair_front_a_02_l, swing1LimitDegrees=20, swing2LimitDegrees=20, twistLimitDegrees=1}
 {
  bone1Name=base_hair_side_a_00_l
  bone2Name=head
  swing1LimitDegrees=45
  swing1Motion=Free
  swing2LimitDegrees=45
  swing2Motion=Free
  twistLimitDegrees=45
  twistMotion=Free
 }
 {bone1Name=dyn_hair_side_a_00_l, bone2Name=base_hair_side_a_00_l, swing1LimitDegrees=25, swing2LimitDegrees=25, twistLimitDegrees=1}
 {
  bone1Name=base_hair_side_b_00_l
  bone2Name=head
  swing1LimitDegrees=45
  swing1Motion=Free
  swing2LimitDegrees=45
  swing2Motion=Free
  twistLimitDegrees=45
  twistMotion=Free
 }
 {bone1Name=dyn_hair_side_b_00_l, bone2Name=base_hair_side_b_00_l, swing1LimitDegrees=45, swing2LimitDegrees=45, twistLimitDegrees=1}
 {bone1Name=dyn_hair_side_b_01_l, bone2Name=dyn_hair_side_b_00_l, swing1LimitDegrees=30, swing2LimitDegrees=30, twistLimitDegrees=1}
]
